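; Player-pawn Blueprint for the FirePetal Wandering Boss, a slow-walking demon NPC that roams or occupies an arena.
; Setup scales health by playlist team size, triggers spawn abilities and sets an AI leash.
; The IsInArena tag drives leashing, minimap icon, damage-threshold gating, death effects and loot tier, with wandering loot drawn from a shuffled DataTable list.
; Replicated weakpoint state drives body-mesh custom primitive data, and helpers handle killfeed and status widget.
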(event EventBeginPlay
  (bind _returnvalue (Animation|GetAnimationInterface))
  (bind _self self)
  (bind _charactermovement (Variables|Character|GetCharacterMovement))
  (bind _datatable_wanderingloottiers (Variables|Default|GetDataTable_WanderingLootTiers))
  (|CallParentFunction)
  (Utilities|FlowControl|Sequence
    (:then_0
      ;[0 0.5365 0.5311 0.5] Initial actor presentation setup.
      (Variables|Default|SetShowShields false)
      (CallFunction|SetNPCAnimSpeedScalar 0.7)
      (Utilities|IsValid _returnvalue
        (:Is%20Valid
          (Locomotion|SetLocomotionPlayRateMultiplier _returnvalue 0.7)
          (Class|FortPlayerAnimationInterface|SetDisableSlopeWarping _returnvalue true)
          (Class|FortPlayerAnimationInterface|SetTurninPlaceAnimPlayRate _returnvalue 0.333)
          (Class|FortPlayerAnimationInterface|SetMovementStyleOverride _returnvalue "Walking")
          (Utilities|IsValid _charactermovement
            (:Is%20Valid
              (Variables|/Script/Engine.CharacterMovementComponent|SetRotationRate_Roll (Math|Rotator|MakeRotator 0.0 0.0 (ScalableFloat|GetValueatLevel (Variables|Tunable|GetDefaultRotateOverrideYaw))) _charactermovement)
              (Ability|Tasks|WaitGameplayTagRemoveFromActor _self "(TagName=\"FirePetalBosses.Demons.IsInArena\")"
                (:then
                  (Ability|Tasks|WaitGameplayTagAddToActor _self "(TagName=\"FirePetalBosses.Demons.IsInArena\")"
                    (:Added
                      (Class|AthenaSpecialActorComponent|SetMiniMapIcon (Variables|Default|GetAthenaSpecialActor) (Variables|Default|GetInArenaMinimapIcon)))))
                (:Removed
                  (Class|AthenaSpecialActorComponent|SetMiniMapIcon (Variables|Default|GetAthenaSpecialActor) (Variables|Default|GetRoamingMinimapIcon)))))))))
    (:then_1
      ;[1 0.3156 0.04167 0.5] Variable Initialization.
      (Variables|InternalData|SetLootTierIndex 0)
      (Variables|InternalData|SetbPreventLootDrop false)
      (bind _asfort_athena_aibot_controller (Utilities|Casting|CastToFortAthenaAIBotController (AI|GetAIController self)))
      (Variables|InternalData|SetAIController _asfort_athena_aibot_controller)
      (Utilities|Array|Clear (Variables|InternalData|GetLootTierArray_Wandering))
      (bind _outrownames (DataTable|GetDataTableRowNames _datatable_wanderingloottiers))
      (for _array_element _outrownames
        (bind _returnvalue_2 (Utilities|GetDataTableRow _datatable_wanderingloottiers _array_element))
        (Utilities|Array|Add (Variables|InternalData|GetLootTierArray_Wandering) (Utilities|Struct|BreakSTRUCTFirePetalWanderingBossLootTableRefs _returnvalue_2)))
      (Utilities|Array|Shuffle (Variables|InternalData|GetLootTierArray_Wandering)))
    (:then_2
      ;[0.37 0.09037 0.8542 0.5] Re-enable use of Rooted Idle Behavior.
      (Networking|SwitchHasAuthority
        (:Authority
          (AlertLevel|BindEventtoOnAlertLevelChangedEvent (Variables|InternalData|GetAIController) (delegate Custom|OnAlertLevelChange))
          ;[0 0.6719 0.1154 0.5] Set Scaled Health Value
          (bind (_gamestate _playlist) (Class|FortAsyncActionCurrentPlaylistReady|PlaylistDataReadyAsync)
            (:OnReady
              (FortPawn|SetMaxHealth (+ (* (ScalableFloat|GetValueatLevel (Variables|InternalData|GetDefaultHealthScalar) (Math|Conversions|ToFloat(Integer) (CallFunction|GetMaxTeamSizefromPlaylist _playlist))) 195.0) (* (* (ScalableFloat|GetValueatLevel (Variables|InternalData|GetDefaultHealthScalar) (Math|Conversions|ToFloat(Integer) (CallFunction|GetMaxTeamSizefromPlaylist _playlist))) 195.0) (ScalableFloat|GetValueatLevel (Variables|Tunable|GetDefaultInvulRatio))))))))))))

(event Custom|OnAlertLevelChange
  (params (BotController :type Fort%20Athena%20AIBot%20Controller%20Object%20Reference) (OldAlertLevel :type EAlertLevel%20Enum :default "Unaware") (NewAlertLevel :type EAlertLevel%20Enum :default "Unaware"))
  ;[0.37 0.09037 0.8542 0.5] Re-enable use of Rooted Idle Behavior.
  (if (not (Variables|Health|GetbIsDBNO))
    (switch Utilities|FlowControl|Switch|SwitchonEAlertLevel NewAlertLevel
      (:Unaware
        (Ability|GameplayEffect|RemoveLooseGameplayTags self (Variables|Tags|GetSyphon%20Idle%20Lock%20Tag))))))

(event EventOnCharacterCustomizationCompleted
  (params (Pawn :type Fort%20Player%20Pawn%20Object%20Reference))
  (|CallParentFunction Pawn)
  (SoundLibrary|AddLibrary (WorldSubsystems|GetSoundLibrarySubsystem) (Weapon|GetCurrentWeapon) "/DemonMaskFire/Assets/Audio/DemonMaskFire_WanderingBoss/SL_WanderingBoss_SFX.SL_WanderingBoss_SFX" :!class /Script/SoundLibrary.SoundLibrarySubsystem))

(event EventPossessed
  (params (NewController :type Controller%20Object%20Reference))
  (bind _returnvalue (Ability|GetAbilitySystemComponent 0))
  (Utilities|FlowControl|DelayUntilNextTick)
  (Utilities|IsValid _returnvalue
    (:Is%20Valid
      (bind _returnvalue_2 (Abilities|TryActivateAbilitiesbyTag _returnvalue (Variables|Tags|GetSpawnAbilityTagContainer) false))
      (if (not _returnvalue_2)
        (Rendering|SetActorHiddenInGame false))
      ;[0.7552 0.006832 0.04441 0.5] On Spawn, set the initial Leash Area accordingly.
      (CallFunction|UpdateLeashArea)
      (|CallParentFunction NewController))
    (:Is%20Not%20Valid
      _)))

(event EventOnDamageServer
  (params (Damage :type Float%20%28single-precision%29) (DamageTags :type Gameplay%20Tag%20Container%20Structure%20%28by%20ref%29) (Momentum :type Vector :default "0, 0, 0") (HitInfo :type Hit%20Result%20Structure%20%28by%20ref%29) (InstigatedBy :type Controller%20Object%20Reference) (DamageCauser :type Actor%20Object%20Reference) (EffectContext :type Gameplay%20Effect%20Context%20Handle%20Structure))
  (bind _bdebugpawn (Variables|Debug|GetbDebugPawn))
  (|CallParentFunction Damage DamageTags Momentum HitInfo InstigatedBy DamageCauser EffectContext)
  (Default|CallOnServerDamagedWanderingBoss DamageTags HitInfo InstigatedBy DamageCauser)
  (if (CallFunction|ShouldProcessThresholds)
    (switch Utilities|FlowControl|Switch|SwitchByRange (FortPawn|HealthPercent self)
      (:MatchesRange_0
        (Development|PrintString "@@@ OnDamageServer [BP_PlayerPawn_FirePetal_WanderingBoss]: 25% - CALLED!" false _bdebugpawn))
      (:MatchesRange_1
        (Development|PrintString "@@@ OnDamageServer [BP_PlayerPawn_FirePetal_WanderingBoss]: 50% - CALLED!" false _bdebugpawn))
      (:MatchesRange_2
        (Development|PrintString "@@@ OnDamageServer [BP_PlayerPawn_FirePetal_WanderingBoss]: 75% - CALLED!" false _bdebugpawn)))
    (GameplayEffects|ApplyGameplayEffectToTarget :self (Ability|GetAbilitySystemComponent 0) :GameplayEffectClass (Variables|InternalData|GetThresholdCrossed_Ammo) :Target (Ability|GetAbilitySystemComponent 0) :Level 0.0)))

(event EventOnDied_Event
  (params (DamagedActor :type Actor%20Object%20Reference) (Damage :type Float%20%28single-precision%29) (InstigatedBy :type Controller%20Object%20Reference) (DamageCauser :type Actor%20Object%20Reference) (HitLocation :type Vector :default "0, 0, 0") (FHitComponent :type Primitive%20Component%20Object%20Reference) (BoneName :type Name :default "None") (Momentum :type Vector :default "0, 0, 0"))
  (|CallParentFunction DamagedActor Damage InstigatedBy DamageCauser HitLocation FHitComponent BoneName Momentum)
  (Development|PrintString "@@@ OnDied_Event [BP_PlayerPawn_FirePetal_WanderingBoss]: CALLED!" false (Variables|Debug|GetbDebugPawn))
  (if (not (Variables|InternalData|GetbPreventLootDrop))
    (bind _bdidactorhavetags (Class|AthenaQuestFunctionLibrary|AthenaCheckActorHasTags self "(GameplayTags=((TagName=\"FirePetalBosses.Demons.IsInArena\")))" true))
    (if _bdidactorhavetags
      (Variables|InternalData|SetCurrentLootTier (Variables|InternalData|GetLootTier_DemonArena))
      (Default|CallOnDead) ; UI Ref for Challenge Bar
      (else
        (Variables|InternalData|SetCurrentLootTier (CallFunction|GetCurrentWanderingLootTier))))
    (Class|AthenaFunctionLibrary|SpawnLootfromTableUsingItemEntry (Variables|InternalData|GetCurrentLootTier) 1 -1 (+ (Pawn|GetNavAgentLocation self) (* (Transformation|GetActorUpVector self) 5e+02)) (Transformation|GetActorForwardVector self) 0 true true false 0 "Other" "NPCElimination")
    ;[0.1142 0.01812 0.15 0.5 13] OnDied - disable the minimap UI immediately FORT-769042
    (Components|Activation|Deactivate (Variables|Default|GetAthenaSpecialActor))))

(event Custom|DisplayKillfeedMessage
  (bind _returnvalue (Fortnite|Athena|GameState|GetGameStateAthena))
  ;[0.1771 0.7649 1 0.5] UI - Called from GA_Relocate_FirePetal_WanderingBoss to display retreating killfeed message
  (Networking|CanExecuteCosmeticEvents
    (:True
      (Utilities|IsValid _returnvalue
        (:Is%20Valid
          (KillFeed|AddKillFeedEntry _returnvalue (Variables|Default|GetKill%20Feed%20Text) 0.0 "(TagName=\"\")"))))))

(event Custom|StartWidgetVerticalOffset
  (params (OffsetY :type Float%20%28double-precision%29) (AnimationTime :type Float%20%28double-precision%29))
  ;[0.4373 1 0 0.5] Status Widget Utilities
  (Class|BPStatusWidgetUtilityCompDefault|UpdateVerticalOffset (Variables|Default|GetBP_StatusWidgetUtilityComp_Default) OffsetY AnimationTime true))

(event Custom|ResetWidgetVerticalOffset
  (params (AnimationTime :type Float%20%28double-precision%29))
  ;[0.4373 1 0 0.5] Status Widget Utilities
  (Class|BPStatusWidgetUtilityCompDefault|ResetVerticalOffset (Variables|Default|GetBP_StatusWidgetUtilityComp_Default) AnimationTime))

(fn GetCurrentWanderingLootTier
  (bind _loottierarray_wandering (Variables|InternalData|GetLootTierArray_Wandering))
  (LocalVariables|Setloc_SelectedTierName (Utilities|Array|Get(acopy) _loottierarray_wandering (Variables|InternalData|GetLootTierIndex)))
  (bind _result (Math|Integer|IncrementInt (Variables|InternalData|GetLootTierIndex))
    (:%20%20
      (if (>= _result (Utilities|Array|Length _loottierarray_wandering))
        (Variables|InternalData|SetLootTierIndex 0))
      (return (LocalVariables|Getloc_SelectedTierName)))))

(fn GetMaxTeamSizeFromPlaylist
  (params (InPlaylistObject :type Fort%20Playlist%20Object%20Reference))
  (bind _output_get (LocalVariables|Setloc_MaxTeamSize (Variables|/Script/FortniteGame.FortPlaylist|GetMaxTeamSize InPlaylistObject)))
  (if (<= _output_get 1)
    (LocalVariables|Setloc_MaxTeamSize 1)
    (elif (>= (LocalVariables|Getloc_MaxTeamSize) 4)
      (LocalVariables|Setloc_MaxTeamSize 4)
      (return (LocalVariables|Getloc_MaxTeamSize)))))

(fn OnRep_bWeakpointsActive
  (if (Variables|InternalData|GetbWeakpointsActive)
    (Default|UpdateWeakpoints true)
    (else
      (Default|UpdateWeakpoints false))))

(fn PlayDeathEffects
  (bind _playeffects (CallFunction|ShouldPlayDeathEffects))
  (if _playeffects
    (|CallParentFunction)))

(fn SetArenaSensors
  (params (InSensorArray :type Array%20of%20B%20Fire%20Petal%20Demon%20Arena%20Island%20Sense%20Volume%20Object%20References))
  (Development|PrintString "@@@ SetArenaSensors [BP_PlayerPawn_FirePetal_WanderingBoss]: CALLED!" false (Variables|Debug|GetbDebugPawn))
  (if (Variables|Debug|GetbDebugPawn)
    (for (_array_element _array_index) (Variables|GetInSensorArray)
      (Development|PrintString (Utilities|String|Append "@@@@@@ SetArenaSensors [BP_PlayerPawn_FirePetal_WanderingBoss]: [" (Utilities|String|ToString(Integer) _array_index) "] = " (Utilities|GetDisplayName _array_element)) false (Variables|Debug|GetbDebugPawn)))
    (else
      _))
  (Variables|InternalData|SetArenaIslandSensorArray (Variables|GetInSensorArray)))

(fn ShouldPlayDeathEffects
  (bind _bdidactorhavetags (Class|AthenaQuestFunctionLibrary|AthenaCheckActorHasTags self "(GameplayTags=((TagName=\"FirePetalBosses.Demons.IsInArena\")))" true))
  (return _bdidactorhavetags))

(fn ShouldProcessThresholds
  (return (and (not (GameplayTags|HasMatchingGameplayTag self "(TagName=\"Ability.WanderingBoss.VoidWarp.Forced\")")) (or (and (GameplayTags|HasMatchingGameplayTag self "(TagName=\"FirePetalBosses.Demons.IsInArena\")") (ScalableFloat|GetValueasBool (Variables|InternalData|GetArenaThresholdEnabled))) (and (not (GameplayTags|HasMatchingGameplayTag self "(TagName=\"FirePetalBosses.Demons.IsInArena\")")) (ScalableFloat|GetValueasBool (Variables|InternalData|GetOpenWorldThresholdEnabled)))))))

(fn UpdateLeashArea
  (bind _leashvolume (Variables|InternalData|GetLeashVolume))
  (Networking|SwitchHasAuthority
    (:Authority
      (bind _bdidactorhavetags (Class|AthenaQuestFunctionLibrary|AthenaCheckActorHasTags self "(GameplayTags=((TagName=\"FirePetalBosses.Demons.IsInArena\")))" true))
      (if _bdidactorhavetags
        (Utilities|IsValid _leashvolume
          (:Is%20Valid
            (Class|FortAthenaAIBotController|SetLeashVolume (Variables|InternalData|GetAIController) _leashvolume)
            (CallFunction|UpdateLeashCenter))
          (:Is%20Not%20Valid
            _))
        (else
          _))
      (CallFunction|UpdateLeashCenter)
      (Class|BPFLAICommonFunctions|SetRadialLeash (Variables|InternalData|GetAIController) (Variables|InternalData|GetCurrentLeashCenter) (ScalableFloat|GetValueatLevel (Variables|Tunable|GetDefaultLeashInnerRadius)) (ScalableFloat|GetValueatLevel (Variables|Tunable|GetDefaultLeashOuterRadius))))))

(fn UpdateLeashCenter
  (bind _aicontroller (Variables|InternalData|GetAIController))
  (bind _returnvalue (Actor|GetComponentbyClass _aicontroller "/Script/FortniteAI.FortAthenaNpcPatrollingComponent"))
  (bind _returnvalue_2 (Class|FortAthenaNpcPatrollingComponent|GetPatrolPath _returnvalue :!class /Script/FortniteAI.FortAthenaNpcPatrollingComponent))
  (bind _returnvalue_3 (Class|FortAthenaPatrolPath|GetPatrolPoint _returnvalue_2 0))
  (Utilities|IsValid _returnvalue
    (:Is%20Valid
      (Utilities|IsValid _returnvalue_2
        (:Is%20Valid
          (Utilities|IsValid _returnvalue_3
            (:Is%20Valid
              (Variables|InternalData|SetCurrentLeashCenter (Transformation|GetActorLocation _returnvalue_3)))
            (:Is%20Not%20Valid
              (Variables|InternalData|SetCurrentLeashCenter (Pawn|GetNavAgentLocation self))))
          (return))
        (:Is%20Not%20Valid
          _)))
    (:Is%20Not%20Valid
      _)))

(fn UpdateWeakpoints
  (params (bTurnOn :type Boolean))
  (bind _returnvalue (Customization|GetSkeletalMeshforPartType "Body"))
  (bind _returnvalue_2 (Utilities|Time|GetGameTimeinSeconds))
  (bind _loc_primitivefloatval (LocalVariables|Getloc_PrimitiveFloatVal))
  (if bTurnOn
    (LocalVariables|Setloc_PrimitiveFloatVal 1.0)
    (else
      (LocalVariables|Setloc_PrimitiveFloatVal 0.0)))
  (Rendering|Material|SetCustomPrimitiveDataVector2 _returnvalue 3 (Math|Vector2D|MakeVector2D _loc_primitivefloatval _returnvalue_2))
  (Rendering|Material|SetCustomPrimitiveDataVector2 _returnvalue 5 (Math|Vector2D|MakeVector2D _loc_primitivefloatval _returnvalue_2)))

(fn ConstructionScript
  (|CallParentFunction))
